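# Physics asset for the Viking 01 hat accessory mesh, set up around a four-bone dynamic beard chain.
# Beard capsules link in sequence from the jaw with tight angular limits, and swing 2 widens toward the tip.
# Capsules on the head, jaw, clavicles and upper spine give the body's shape, with head, clavicle and spine joints left free in swing and twist.
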PhysicsAsset /Game/Accessories/Hats/M_MED_Viking_01_Hat/Meshes/M_MED_Viking_01_Hat_FE_Physics
bodies {
 C_jaw [{shapeType=Capsule, length=3, radius=2.75, center=<0 2.7244 -1.7836>, rotation=<rot 90 0 0>}]
 clavicle_l [{shapeType=Capsule, length=5, radius=5}]
 clavicle_r [{shapeType=Capsule, length=5, radius=5}]
 dyn_Beard_01 [{shapeType=Capsule, length=2.5, radius=2.5, center=<0 -0.1595 -4.2094>}]
 dyn_Beard_02 [{shapeType=Capsule, length=2.25, radius=2.75, center=<0 0.0799 -4.2116>}]
 dyn_Beard_03 [{shapeType=Capsule, length=2, radius=3, center=<0 0.2965 -4.2019>}]
 dyn_Beard_04 [{shapeType=Capsule, length=2, radius=3, center=<0 0.2965 -4.2019>}]
 head [{shapeType=Capsule, length=5, radius=5, center=<10.0929 3.8505 -0.0001>, rotation=<rot 90 0 0>}]
 spine_04 [{shapeType=Capsule, length=15.5, radius=13.75, center=<3.3533 1.496 0>, rotation=<rot 0 -90 0>}]
 spine_05 [{shapeType=Capsule, length=15, radius=14, center=<2.0517 -1.3532 0>}]
}
constraints [
 {bone1Name=C_jaw, bone2Name=head, swing1LimitDegrees=45, swing2LimitDegrees=45, twistLimitDegrees=45}
 {
  bone1Name=head
  bone2Name=spine_05
  swing1LimitDegrees=45
  swing1Motion=Free
  swing2LimitDegrees=45
  swing2Motion=Free
  twistLimitDegrees=45
  twistMotion=Free
 }
 {
  bone1Name=clavicle_l
  bone2Name=spine_05
  swing1LimitDegrees=45
  swing1Motion=Free
  swing2LimitDegrees=45
  swing2Motion=Free
  twistLimitDegrees=45
  twistMotion=Free
 }
 {
  bone1Name=clavicle_r
  bone2Name=spine_05
  swing1LimitDegrees=45
  swing1Motion=Free
  swing2LimitDegrees=45
  swing2Motion=Free
  twistLimitDegrees=45
  twistMotion=Free
 }
 {bone1Name=dyn_Beard_01, bone2Name=C_jaw, swing1LimitDegrees=5, swing2LimitDegrees=10, twistLimitDegrees=30}
 {bone1Name=dyn_Beard_02, bone2Name=dyn_Beard_01, swing1LimitDegrees=5, swing2LimitDegrees=12, twistLimitDegrees=30}
 {bone1Name=dyn_Beard_03, bone2Name=dyn_Beard_02, swing1LimitDegrees=5, swing2LimitDegrees=15, twistLimitDegrees=30}
 {bone1Name=dyn_Beard_04, bone2Name=dyn_Beard_03, swing1LimitDegrees=5, swing2LimitDegrees=18, twistLimitDegrees=30}
 {
  bone1Name=spine_05
  bone2Name=spine_04
  swing1LimitDegrees=45
  swing1Motion=Free
  swing2LimitDegrees=45
  swing2Motion=Free
  twistLimitDegrees=45
  twistMotion=Free
 }
]
