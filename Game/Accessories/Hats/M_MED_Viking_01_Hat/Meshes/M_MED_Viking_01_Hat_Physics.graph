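# Physics asset for the medieval Viking 01 hat accessory, covering its dangling beard and hair bones plus the upper body around them.
# Capsule bodies on the spine, clavicles, head and jaw form the base, with free angular joints between spine, clavicles and head.
# A four-link center beard chain hangs from the jaw, while two-link side beard and three hair chains hang from the head.
# These dynamic chains have angular limits for swing and twist.
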PhysicsAsset /Game/Accessories/Hats/M_MED_Viking_01_Hat/Meshes/M_MED_Viking_01_Hat_Physics
bodies {
 C_jaw [{shapeType=Capsule, length=3, radius=2.75, center=<0 2.7244 -1.7836>, rotation=<rot 90 0 0>}]
 clavicle_l [{shapeType=Capsule, length=5, radius=5}]
 clavicle_r [{shapeType=Capsule, length=5, radius=5}]
 dyn_Beard_01 [{shapeType=Capsule, length=2, radius=3, center=<0 -0.1595 -4.2094>}]
 dyn_Beard_02 [{shapeType=Capsule, length=2, radius=3, center=<0 0.0799 -4.2116>}]
 dyn_Beard_03 [{shapeType=Capsule, length=2, radius=3, center=<0 0.2965 -4.2019>}]
 dyn_Beard_04 [{shapeType=Capsule, length=2, radius=3, center=<0 0.2965 -4.2019>}]
 dyn_Beard_l_01 [{shapeType=Capsule, length=2.122, radius=0.7966, center=<0.0658 -0.0207 -3.1749>, rotation=<rot -9.7878 -90 -2.6804>}]
 dyn_Beard_l_02 [{shapeType=Capsule, length=3.1915, radius=0.7173, center=<0.0083 0.1054 -3.01>, rotation=<rot -0.0297 -90 -1.7106>}]
 dyn_Beard_r_01 [{shapeType=Capsule, length=2.122, radius=0.7966, center=<-0.0659 0.0208 3.1753>, rotation=<rot -9.7878 -90 -2.6804>}]
 dyn_Beard_r_02 [{shapeType=Capsule, length=3.1915, radius=0.7173, center=<-0.0083 -0.1052 3.0104>, rotation=<rot -0.0297 -90 -1.7106>}]
 dyn_vik_hair_01 [{shapeType=Capsule, length=1, radius=1}]
 dyn_vik_hair_01_l [{shapeType=Capsule, length=1, radius=1}]
 dyn_vik_hair_01_r [{shapeType=Capsule, length=1, radius=1}]
 dyn_vik_hair_02 [{shapeType=Capsule, length=1, radius=1}]
 dyn_vik_hair_02_l [{shapeType=Capsule, length=1, radius=1}]
 dyn_vik_hair_02_r [{shapeType=Capsule, length=1, radius=1}]
 head [{shapeType=Capsule, length=5, radius=5, center=<10.0929 3.8505 -0.0001>, rotation=<rot 90 0 0>}]
 spine_04 [{shapeType=Capsule, length=15.5, radius=13.75, center=<3.3533 1.496 0>, rotation=<rot 0 -90 0>}]
 spine_05 [{shapeType=Capsule, length=14.75, radius=14.25, center=<2.0517 -1.3532 0>}]
}
constraints [
 {bone1Name=C_jaw, bone2Name=head, swing1LimitDegrees=45, swing2LimitDegrees=45, twistLimitDegrees=45}
 {bone1Name=dyn_Beard_r_01, bone2Name=head, swing1LimitDegrees=5, swing2LimitDegrees=30, twistLimitDegrees=30}
 {bone1Name=dyn_Beard_r_02, bone2Name=dyn_Beard_r_01, swing1LimitDegrees=5, swing2LimitDegrees=30, twistLimitDegrees=30}
 {bone1Name=dyn_Beard_l_01, bone2Name=head, swing1LimitDegrees=5, swing2LimitDegrees=30, twistLimitDegrees=30}
 {bone1Name=dyn_Beard_l_02, bone2Name=dyn_Beard_l_01, swing1LimitDegrees=5, swing2LimitDegrees=30, twistLimitDegrees=30}
 {bone1Name=dyn_vik_hair_01, bone2Name=head, swing1LimitDegrees=20, swing2LimitDegrees=20, twistLimitDegrees=10}
 {bone1Name=dyn_vik_hair_02, bone2Name=dyn_vik_hair_01, swing1LimitDegrees=20, swing2LimitDegrees=20, twistLimitDegrees=10}
 {bone1Name=dyn_vik_hair_01_l, bone2Name=head, swing1LimitDegrees=20, swing2LimitDegrees=20, twistLimitDegrees=10}
 {bone1Name=dyn_vik_hair_02_l, bone2Name=dyn_vik_hair_01_l, swing1LimitDegrees=20, swing2LimitDegrees=20, twistLimitDegrees=10}
 {bone1Name=dyn_vik_hair_01_r, bone2Name=head, swing1LimitDegrees=20, swing2LimitDegrees=20, twistLimitDegrees=10}
 {bone1Name=dyn_vik_hair_02_r, bone2Name=dyn_vik_hair_01_r, swing1LimitDegrees=20, swing2LimitDegrees=20, twistLimitDegrees=10}
 {
  bone1Name=head
  bone2Name=spine_05
  swing1LimitDegrees=45
  swing1Motion=Free
  swing2LimitDegrees=45
  swing2Motion=Free
  twistLimitDegrees=45
  twistMotion=Free
 }
 {
  bone1Name=clavicle_l
  bone2Name=spine_05
  swing1LimitDegrees=45
  swing1Motion=Free
  swing2LimitDegrees=45
  swing2Motion=Free
  twistLimitDegrees=45
  twistMotion=Free
 }
 {
  bone1Name=clavicle_r
  bone2Name=spine_05
  swing1LimitDegrees=45
  swing1Motion=Free
  swing2LimitDegrees=45
  swing2Motion=Free
  twistLimitDegrees=45
  twistMotion=Free
 }
 {bone1Name=dyn_Beard_01, bone2Name=C_jaw, swing1LimitDegrees=5, swing2LimitDegrees=30, twistLimitDegrees=30}
 {bone1Name=dyn_Beard_02, bone2Name=dyn_Beard_01, swing1LimitDegrees=5, swing2LimitDegrees=30, twistLimitDegrees=30}
 {bone1Name=dyn_Beard_03, bone2Name=dyn_Beard_02, swing1LimitDegrees=5, swing2LimitDegrees=30, twistLimitDegrees=30}
 {bone1Name=dyn_Beard_04, bone2Name=dyn_Beard_03, swing1LimitDegrees=5, swing2LimitDegrees=30, twistLimitDegrees=30}
 {
  bone1Name=spine_05
  bone2Name=spine_04
  swing1LimitDegrees=45
  swing1Motion=Free
  swing2LimitDegrees=45
  swing2Motion=Free
  twistLimitDegrees=45
  twistMotion=Free
 }
]
