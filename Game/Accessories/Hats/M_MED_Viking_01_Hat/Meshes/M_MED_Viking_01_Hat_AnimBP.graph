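# Animation Blueprint for the Viking 01 hat accessory, copying the attached parent's pose and adding secondary motion to its beard and hair bones.
# bIsInFrontEnd selects a head-space RigidBody simulation with the front-end physics asset, or otherwise an AnimDynamics chain on the center beard.
# After that blend, AnimDynamics chains drive the side beard and three hair strands, using spherical limits on spine and neck as body collision.
# Several simulation and sequence nodes are left unconnected.
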AnimBlueprint /Game/Accessories/Hats/M_MED_Viking_01_Hat/Meshes/M_MED_Viking_01_Hat_AnimBP
target_skeleton=&/M_MED_Viking_01_Hat_Skeleton
graphs [
 AnimGraph {nodes [
   "Anim Dynamics - Chain: dyn_Beard_l_01 -> dyn_Beard_l_02"(Alpha=1, ExternalForce="0,0,0") {
    angularDampingOverride=0.99
    angularSpringConstant=5000
    bAngularSpring=true
    bChain=true
    bGravityOverrideInSimSpace=true
    bOverrideAngularDamping=true
    numSolverIterationsPostUpdate=4
    numSolverIterationsPreUpdate=16
    simulationSpace=Actor
    alphaBoolBlend {blendOption=Linear}
    boundBone {boneName=dyn_Beard_l_01}
    chainEnd {boneName=dyn_Beard_l_02}
    physicsBodyDefinitions [{
      sphereCollisionRadius=0
      boundBone {boneName=dyn_Beard_l_01}
      boxExtents=<10 20 20>
      constraintSetup {angularLimitsMax=<5 20 20>, angularLimitsMin=<-5 -20 -15>, angularTarget=<1 0 0>}
      localJointOffset=<10 0 0>
     }]
   }
   "Anim Dynamics - Chain: dyn_Beard_01 -> dyn_Beard_02"(Alpha=1, ExternalForce="0,0,0") {
    angularDampingOverride=0.999
    angularSpringConstant=75
    bAngularSpring=true
    bChain=true
    bGravityOverrideInSimSpace=true
    bOverrideAngularDamping=true
    bUseSphericalLimits=true
    numSolverIterationsPostUpdate=4
    numSolverIterationsPreUpdate=16
    alphaBoolBlend {blendOption=Linear}
    boundBone {boneName=dyn_Beard_01}
    chainEnd {boneName=dyn_Beard_02}
    physicsBodyDefinitions [{
      collisionType=CustomSphere
      sphereCollisionRadius=2
      boundBone {boneName=dyn_Beard_01}
      boxExtents=<50 50 30>
      constraintSetup {angularTargetAxis=AxisZ, twistAxis=AxisZ, angularLimitsMax=<20 20 5>, angularLimitsMin=<-5 -20 -5>, angularTarget=<0 0 1>}
      localJointOffset=<0 0 -2>
     }]
    sphericalLimits [
     {limitRadius=14.25, drivingBone {boneName=spine_05}, sphereLocalOffset=<0 0 0>}
     {limitRadius=14.25, drivingBone {boneName=spine_05}, sphereLocalOffset=<0 0 -10>}
     {limitRadius=14.25, drivingBone {boneName=spine_05}, sphereLocalOffset=<0 0 10>}
    ]
   }
   RigidBody() {
    overridePhysicsAsset=&/M_MED_Viking_01_Hat_Physics
    alphaBoolBlend {blendOption=Linear}
    componentLinearAccScale=<5 5 5>
    componentLinearVelScale=<1 1 1>
   }
   n1 = "Anim Dynamics - Chain: dyn_Beard_01 -> dyn_Beard_04"(ComponentPose=Local To Component(LocalPose=Copy Pose From Mesh() {bUseAttachedParent=true}), Alpha=1, ExternalForce="0,0,0") {
    angularDampingOverride=0.999
    angularSpringConstant=100
    bAngularSpring=true
    bChain=true
    bGravityOverrideInSimSpace=true
    bOverrideAngularDamping=true
    bUseSphericalLimits=true
    numSolverIterationsPostUpdate=4
    numSolverIterationsPreUpdate=16
    alphaBoolBlend {blendOption=Linear}
    boundBone {boneName=dyn_Beard_01}
    chainEnd {boneName=dyn_Beard_04}
    physicsBodyDefinitions [{
      collisionType=CustomSphere
      sphereCollisionRadius=3
      boundBone {boneName=dyn_Beard_01}
      boxExtents=<50 50 30>
      constraintSetup {angularTargetAxis=AxisZ, twistAxis=AxisZ, angularLimitsMax=<20 20 5>, angularLimitsMin=<-20 -20 -5>, angularTarget=<0 0 1>}
      localJointOffset=<0 0 -2>
     }]
    sphericalLimits [
     {limitRadius=14.25, drivingBone {boneName=spine_05}, sphereLocalOffset=<0 0 0>}
     {limitRadius=14.25, drivingBone {boneName=spine_05}, sphereLocalOffset=<0 0 -10>}
     {limitRadius=14.25, drivingBone {boneName=spine_05}, sphereLocalOffset=<0 0 10>}
    ]
   }
   n2 = Component To Local(ComponentPose=@n1)
   n3 = RigidBody(ComponentPose=Local To Component(LocalPose=Copy Pose From Mesh() {bUseAttachedParent=true}), Alpha=0.75) {
    bClampLinearTranslationLimitToRefPose=true
    bOverrideWorldGravity=true
    overridePhysicsAsset=&/M_MED_Viking_01_Hat_FE_Physics
    simulationSpace=BaseBoneSpace
    alphaBoolBlend {blendOption=Linear}
    baseBoneRef {boneName=head}
    componentLinearAccScale=<0.2 0.2 0.2>
    componentLinearVelScale=<0.1 0.1 0.1>
    overrideWorldGravity {z=-980}
   }
   n4 = Component To Local(ComponentPose=@n3)
   n5 = Get bIsInFrontEnd()
   n6 = "Anim Dynamics - Chain: dyn_Beard_l_01 -> dyn_Beard_l_02"(
    ComponentPose=Local To Component(LocalPose=Blend Poses by bool(BlendPose_0=@n4, BlendPose_1=@n2, bActiveValue=@n5) {blendType=Linear})
    Alpha=1
    ExternalForce="0,0,0"
   ) {
    angularDampingOverride=0.999
    angularSpringConstant=100
    bAngularSpring=true
    bChain=true
    bGravityOverrideInSimSpace=true
    bOverrideAngularDamping=true
    bUseSphericalLimits=true
    numSolverIterationsPostUpdate=4
    numSolverIterationsPreUpdate=16
    alphaBoolBlend {blendOption=Linear}
    boundBone {boneName=dyn_Beard_l_01}
    chainEnd {boneName=dyn_Beard_l_02}
    physicsBodyDefinitions [{
      collisionType=CustomSphere
      sphereCollisionRadius=2
      boundBone {boneName=dyn_Beard_l_01}
      boxExtents=<50 50 30>
      constraintSetup {angularTargetAxis=AxisZ, twistAxis=AxisZ, angularLimitsMax=<20 20 5>, angularLimitsMin=<-20 -20 -5>, angularTarget=<0 0 1>}
      localJointOffset=<0 0 -2>
     }]
    sphericalLimits [
     {limitRadius=14.25, drivingBone {boneName=spine_05}, sphereLocalOffset=<0 0 0>}
     {limitRadius=14.25, drivingBone {boneName=spine_05}, sphereLocalOffset=<0 0 -10>}
     {limitRadius=14.25, drivingBone {boneName=spine_05}, sphereLocalOffset=<0 0 10>}
    ]
   }
   n7 = "Anim Dynamics - Chain: dyn_Beard_r_01 -> dyn_Beard_r_02"(ComponentPose=@n6, Alpha=1, ExternalForce="0,0,0") {
    angularDampingOverride=0.999
    angularSpringConstant=100
    bAngularSpring=true
    bChain=true
    bGravityOverrideInSimSpace=true
    bOverrideAngularDamping=true
    bUseSphericalLimits=true
    numSolverIterationsPostUpdate=4
    numSolverIterationsPreUpdate=16
    alphaBoolBlend {blendOption=Linear}
    boundBone {boneName=dyn_Beard_r_01}
    chainEnd {boneName=dyn_Beard_r_02}
    physicsBodyDefinitions [{
      collisionType=CustomSphere
      sphereCollisionRadius=2
      boundBone {boneName=dyn_Beard_r_01}
      boxExtents=<50 50 30>
      constraintSetup {angularTargetAxis=AxisZ, twistAxis=AxisZ, angularLimitsMax=<20 20 5>, angularLimitsMin=<-20 -20 -5>, angularTarget=<0 0 1>}
      localJointOffset=<0 0 2>
     }]
    sphericalLimits [
     {limitRadius=14.25, drivingBone {boneName=spine_05}, sphereLocalOffset=<0 0 0>}
     {limitRadius=14.25, drivingBone {boneName=spine_05}, sphereLocalOffset=<0 0 -10>}
     {limitRadius=14.25, drivingBone {boneName=spine_05}, sphereLocalOffset=<0 0 10>}
    ]
   }
   n8 = "Anim Dynamics - Chain: dyn_vik_hair_01 -> dyn_vik_hair_02"(ComponentPose=@n7, Alpha=1, ExternalForce="0,0,0") {
    angularDampingOverride=0.98
    angularSpringConstant=100
    bAngularSpring=true
    bChain=true
    bGravityOverrideInSimSpace=true
    bOverrideAngularDamping=true
    bUseSphericalLimits=true
    numSolverIterationsPostUpdate=4
    numSolverIterationsPreUpdate=16
    alphaBoolBlend {blendOption=Linear}
    boundBone {boneName=dyn_vik_hair_01}
    chainEnd {boneName=dyn_vik_hair_02}
    physicsBodyDefinitions [{
      collisionType=CustomSphere
      sphereCollisionRadius=1
      boundBone {boneName=dyn_vik_hair_01}
      boxExtents=<20 30 30>
      constraintSetup {angularTargetAxis=AxisZ, angularLimitsMax=<5 20 5>, angularLimitsMin=<-5 -20 -20>, angularTarget=<0 0 1>}
      localJointOffset=<2 0 0>
     }]
    sphericalLimits [{limitRadius=12, drivingBone {boneName=neck_01}, sphereLocalOffset=<0 0 0>}]
   }
   n9 = "Anim Dynamics - Chain: dyn_vik_hair_01_l -> dyn_vik_hair_02_l"(ComponentPose=@n8, Alpha=1, ExternalForce="0,0,0") {
    angularDampingOverride=0.98
    angularSpringConstant=100
    bAngularSpring=true
    bChain=true
    bGravityOverrideInSimSpace=true
    bOverrideAngularDamping=true
    bUseSphericalLimits=true
    numSolverIterationsPostUpdate=4
    numSolverIterationsPreUpdate=16
    alphaBoolBlend {blendOption=Linear}
    boundBone {boneName=dyn_vik_hair_01_l}
    chainEnd {boneName=dyn_vik_hair_02_l}
    physicsBodyDefinitions [{
      collisionType=CustomSphere
      sphereCollisionRadius=1.1
      boundBone {boneName=dyn_vik_hair_01_l}
      boxExtents=<20 30 30>
      constraintSetup {angularTargetAxis=AxisZ, angularLimitsMax=<5 20 5>, angularLimitsMin=<-5 -20 -20>, angularTarget=<0 0 1>}
      localJointOffset=<2 0 0>
     }]
    sphericalLimits [{limitRadius=27, drivingBone {boneName=spine_05}, sphereLocalOffset=<-3 8 2>}]
   }
   n10 = "Anim Dynamics - Chain: dyn_vik_hair_01_r -> dyn_vik_hair_02_r"(ComponentPose=@n9, Alpha=1, ExternalForce="0,0,0") {
    angularDampingOverride=0.98
    angularSpringConstant=100
    bAngularSpring=true
    bChain=true
    bGravityOverrideInSimSpace=true
    bOverrideAngularDamping=true
    bUseSphericalLimits=true
    numSolverIterationsPostUpdate=4
    numSolverIterationsPreUpdate=16
    alphaBoolBlend {blendOption=Linear}
    boundBone {boneName=dyn_vik_hair_01_r}
    chainEnd {boneName=dyn_vik_hair_02_r}
    physicsBodyDefinitions [{
      collisionType=CustomSphere
      sphereCollisionRadius=1.1
      boundBone {boneName=dyn_vik_hair_01_r}
      boxExtents=<20 30 30>
      constraintSetup {angularTargetAxis=AxisZ, angularLimitsMax=<5 20 5>, angularLimitsMin=<-5 -20 -20>, angularTarget=<0 0 1>}
      localJointOffset=<-2 0 0>
     }]
    sphericalLimits [{limitRadius=27, drivingBone {boneName=spine_05}, sphereLocalOffset=<-3 8 -2>}]
   }
   Output Pose(Result=Component To Local(ComponentPose=@n10)) {layerGroup=None}
   Sequence Player 'LookingUp'() {sequence=/Game/Accessories/Hats/M_MED_Viking_01_Hat/Animations/LookingUp}
   Sequence Player 'Running_Around_Ingame_02'() {sequence=/Game/Accessories/Hats/M_MED_Viking_01_Hat/Animations/Running_Around_Ingame_02}
  ]}
 EventGraph {nodes [
   TryGetPawnOwner()
   Event BlueprintUpdateAnimation()
  ]}
]
variables []
